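# Physics asset for the Cardboard Crew Holiday cape back accessory, defining collision and joint limits for its dynamic cape bones.
# Single capsule bodies sit on the backpack root, a master-rotate bone and four cape bones.
# Constraints link the cape bones in a chain from the master-rotate bone, allowing much more swing on one axis than in the other swing or twist.
# The backpack root body has no constraint.
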PhysicsAsset /BRCosmetics/Accessories/FORT_Capes/M_MED_Cardboard_Crew_Holiday_Cape/Meshes/M_MED_Cardboard_Crew_Holiday_Cape_dyn
bodies {
 backpack_root [{shapeType=Capsule, length=4, radius=3, rotation=<rot 90 0 0>}]
 dyn_cape01 [{shapeType=Capsule, length=15.5375, radius=3.8415, center=<3.1974 -0.5114 -0.7822>, rotation=<rot -1.8592 -90 0.1332>}]
 dyn_cape01_masterRotate [{shapeType=Capsule, length=0.505, radius=2.505, center=<-1.7824 0.0001 0>, rotation=<rot 0 -90 0>}]
 dyn_cape02 [{shapeType=Capsule, length=16, radius=4, center=<2.8238 0 0>}]
 dyn_cape03 [{shapeType=Capsule, length=16, radius=3.75, center=<2.5534 0 0>}]
 dyn_cape04 [{shapeType=Capsule, length=16, radius=3.75, center=<0.8701 -1.3957 0>}]
}
constraints [
 {bone1Name=dyn_cape04, bone2Name=dyn_cape03, swing1LimitDegrees=25, swing2LimitDegrees=5, twistLimitDegrees=5}
 {bone1Name=dyn_cape02, bone2Name=dyn_cape01, swing1LimitDegrees=20, swing2LimitDegrees=5, twistLimitDegrees=4}
 {bone1Name=dyn_cape03, bone2Name=dyn_cape02, swing1LimitDegrees=25, swing2LimitDegrees=7, twistLimitDegrees=4}
 {bone1Name=dyn_cape01, bone2Name=dyn_cape01_masterRotate, swing1LimitDegrees=35, swing2LimitDegrees=8, twistLimitDegrees=5}
]
