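# Animation Blueprint for the Cardboard Crew holiday cape accessory, posing and simulating the cloak's bones on the character.
# It copies the parent mesh pose, adds skydive, backward-movement and falling rotations, then runs a RigidBody simulation in backpack-root space.
# When rigid body is disabled, fixed bone rotations replace the cached result's cape pose.
# The event graph drives these rotations and gravity, zeroed while swim-sprinting; LowLODAlpha is computed but unused.
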AnimBlueprint /BRCosmetics/Accessories/FORT_Capes/M_MED_Cardboard_Crew_Holiday_Cape/Meshes/M_MED_Cardboard_Crew_Holiday_Cape_AnimBP
target_skeleton=&/M_MED_Cardboard_Crew_Holiday_Cape_Skeleton
graphs [
 AnimGraph {nodes [
   "Transform (Modify) Bone - Bone: dyn_cape01"(Alpha=1, Translation="0,0,0", Rotation="0,0,0", Scale="1,1,1") {rotationMode=BMM_Additive, rotationSpace=BCS_BoneSpace, alphaBoolBlend {blendOption=Linear}, boneToModify {boneName=dyn_cape01}}
   n1 = "Transform (Modify) Bone - Bone: dyn_cape01_masterRotate"(
    ComponentPose=Local To Component(LocalPose=Use cached pose 'TransformCache'())
    Alpha=1
    Translation="0,0,0"
    Rotation="0,-20,0"
    Scale="1,1,1"
   ) {
    rotationMode=BMM_Replace
    rotationSpace=BCS_BoneSpace
    alphaBoolBlend {blendOption=Linear}
    boneToModify {boneName=dyn_cape01_masterRotate}
    rotation {yaw=-20}
   }
   n2 = "Transform (Modify) Bone - Bone: dyn_cape02"(ComponentPose=@n1, Alpha=1, Translation="0,0,0", Rotation="0,0,0", Scale="1,1,1") {
    rotationMode=BMM_Replace
    rotationSpace=BCS_BoneSpace
    alphaBoolBlend {blendOption=Linear}
    alphaScaleBiasClamp {interpSpeedDecreasing=0, interpSpeedIncreasing=0}
    boneToModify {boneName=dyn_cape02}
   }
   n3 = "Transform (Modify) Bone - Bone: dyn_cape03"(ComponentPose=@n2, Alpha=1, Translation="0,0,0", Rotation="0,0,0", Scale="1,1,1") {
    rotationMode=BMM_Replace
    rotationSpace=BCS_BoneSpace
    alphaBoolBlend {blendOption=Linear}
    alphaScaleBiasClamp {interpSpeedDecreasing=0, interpSpeedIncreasing=0}
    boneToModify {boneName=dyn_cape03}
   }
   Output Pose(Result=Blend Poses by bool(BlendPose_0=Use cached pose 'TransformCache'(), BlendPose_1=Component To Local(ComponentPose=@n3), bActiveValue=Get bRigidBodyEnabled()) {blendType=Linear}) {layerGroup=None}
   n4 = "Transform (Modify) Bone - Bone: dyn_cape01_masterRotate"(
    ComponentPose=Local To Component(LocalPose=Copy Pose From Mesh() {bUseAttachedParent=true})
    Alpha=1
    Translation="0,0,0"
    Rotation=Get SkydiveAlpha() -> (SkydiveAlpha:Rotator)
    Scale="1,1,1"
   ) {rotationMode=BMM_Additive, rotationSpace=BCS_BoneSpace, alphaBoolBlend {blendOption=Linear}, boneToModify {boneName=dyn_cape01_masterRotate}}
   n5 = "Transform (Modify) Bone - Bone: dyn_cape01_masterRotate"(ComponentPose=@n4, Alpha=Get BwdMovement() -> (BwdMovement:float), Translation="0,0,0", Rotation="0,-5,0", Scale="1,1,1") {
    rotationMode=BMM_Additive
    rotationSpace=BCS_BoneSpace
    alphaBoolBlend {blendOption=Linear}
    alphaScaleBiasClamp {bInterpResult=true}
    boneToModify {boneName=dyn_cape01_masterRotate}
    rotation {yaw=-5}
   }
   n6 = Get JumpAlpha_0() -> (JumpAlpha_0:double)
   n7 = Get JumpAlpha() -> (JumpAlpha:Rotator)
   n8 = "Transform (Modify) Bone - Bone: dyn_cape02"(ComponentPose=@n5, Alpha=@n6, Translation="0,0,0", Rotation=@n7, Scale="1,1,1") {
    rotationMode=BMM_Additive
    rotationSpace=BCS_BoneSpace
    alphaBoolBlend {blendOption=Linear}
    alphaScaleBiasClamp {bInterpResult=true}
    boneToModify {boneName=dyn_cape02}
   }
   n9 = "Transform (Modify) Bone - Bone: dyn_cape03"(ComponentPose=@n8, Alpha=@n6, Translation="0,0,0", Rotation=@n7, Scale="1,1,1") {
    rotationMode=BMM_Additive
    rotationSpace=BCS_BoneSpace
    alphaBoolBlend {blendOption=Linear}
    alphaScaleBiasClamp {bInterpResult=true}
    boneToModify {boneName=dyn_cape03}
   }
   n10 = RigidBody!(ComponentPose=@n9, Alpha, OverrideWorldGravity:Vector=Get GravityOverride(), ExternalForce) -> (Pose) {
    bClampLinearTranslationLimitToRefPose=true
    bOverrideWorldGravity=true
    overridePhysicsAsset=&/M_MED_Cardboard_Crew_Holiday_Cape_dyn
    simulationSpace=BaseBoneSpace
    alphaBoolBlend {blendOption=Linear}
    baseBoneRef {boneName=backpack_root}
    componentAppliedLinearAccClamp=<3000 3000 3000>
    componentLinearAccScale=<0.45 0.3 0.35>
    componentLinearVelScale=<0.45 0.3 0.35>
    overrideWorldGravity {z=-980}
   }
   Save cached pose 'TransformCache'(Pose=Component To Local(ComponentPose=@n10)) {cacheName=TransformCache}
   Get Is Low LOD() -> (Is Low LOD:bool)
   Get LowLODAlpha() -> (LowLODAlpha:Rotator)
  ]}
 EventGraph {nodes [
   Event BlueprintUpdateAnimation()
   > n1 = Sequence!(execute) -> (then_0, then_1, then_2:Exec, then_3:Exec, then_4:Exec, then_5:Exec)
   > n2 = Branch(Condition=Get bIsSwimSprinting() -> (bIsSwimSprinting:bool))
   > Set GravityOverride(GravityOverride="0, 0, 0")
   n3 = Branch(execute=@n2.else, Condition=Get bIsSkyDiving())
   > n4 = Branch(Condition=Get bIsParachuteOpened())
   Set GravityOverride(execute=[@n3.else, @n4.then], GravityOverride="0, 0,-980")
   n5 = Branch(execute=@n4.else, Condition="float > float"(A=Get PawnVelocityZ(), B=40) {is=title K2Node_CallFunction})
   > Set GravityOverride(GravityOverride="0, 0,-980")
   Set GravityOverride(execute=@n5.else, GravityOverride="0, 0,-980")
   Set SkydiveAlpha(
    execute=@n1.then_1
    SkydiveAlpha:Rotator=SelectRotator!(A:Rotator="0,-25,0", B:Rotator="0,20, 0", bPickA:bool=Get bIsSkyDiving()) -> (ReturnValue:Rotator)
   ) -> (Output_Get:Rotator)
   n6 = NOT Boolean(A=Get Is Low LOD())
   n7 = Get bIsFalling()
   Set JumpAlpha(execute=@n1.then_2, JumpAlpha:Rotator=SelectRotator(A="0,-10,0", B="0,5, 0", bPickA=AND Boolean(A=@n7, B=@n6))) -> (Output_Get:Rotator)
   > n8 = PlayerMovement(AnimInstance=Self-Reference(), XDotProductCoefficient=1, YDotProductCoefficient=1, ZDotProductCoefficient=1)
   > Set BwdMovement(BwdMovement:float=MapRangeClamped(Value=@n8.YMovement, InRangeA=0, InRangeB=-1)) -> (Output_Get:float)
   Set Is Low LOD(
    execute=@n1.then_3
    Is Low LOD:bool="integer >= integer"!(A:int=Get CurrentLOD(), B:int=1) -> (ReturnValue:bool) {is=title K2Node_CallFunction}
   ) -> (Output_Get:bool)
   Set LowLODAlpha(execute=@n1.then_4, LowLODAlpha:Rotator=SelectRotator(A="0,-23,0", B="0,-5, 0", bPickA=Get Is Low LOD())) -> (Output_Get:Rotator)
   Event OnBeginFalling()
   > Set JumpAlpha_0(JumpAlpha_0:double=1) -> (Output_Get:double)
   Event OnEndFalling()
   > Set JumpAlpha_0(JumpAlpha_0=0)
   TryGetPawnOwner()
   TryGetPawnOwner()
  ]}
]
variables [
 BwdMovement {type=number}
 Character Part Anim Instance {type=CustomCharacterPartAnimInstance, default=null}
 GravityOverride {type=Vector}
 Is Low LOD {type=boolean}
 JumpAlpha {type=Rotator}
 JumpAlpha_0 {type=number}
 LowLODAlpha {type=Rotator}
 SkydiveAlpha {type=Rotator}
]
